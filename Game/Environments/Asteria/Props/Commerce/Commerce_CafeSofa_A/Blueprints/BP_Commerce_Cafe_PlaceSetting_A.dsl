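; Café place setting prop Blueprint for the Asteria commerce environment, built on a parent class and using a café sofa mesh component.
; The construction script runs the parent's logic, then applies a fixed dark grey color to the ISM_CafeSofa component through SetColoronSecondaryMesh.
; When the actor is destroyed, it destroys its TableCollision component.
; BeginPlay, overlap and Tick events are empty stubs.
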(event EventBeginPlay)

(event Collision|EventActorBeginOverlap
  (params (OtherActor :type Actor%20Object%20Reference)))

(event EventTick
  (params (DeltaSeconds :type Float%20%28single-precision%29)))

(event EventDestroyed
  (Components|DestroyComponent (Variables|Default|GetTableCollision)))

(fn ConstructionScript
  (|CallParentFunction)
  (Default|SetColoronSecondaryMesh "(R=0.031896,G=0.031896,B=0.031896,A=1.000000)" (Variables|Default|GetISM_CafeSofa)))
