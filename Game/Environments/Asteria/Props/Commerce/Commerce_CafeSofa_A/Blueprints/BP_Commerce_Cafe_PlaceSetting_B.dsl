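; BP_Commerce_Cafe_PlaceSetting_B is a cafe place setting prop Blueprint in the Asteria commerce set, built on a parent class with a sofa mesh component.
; Its construction script runs the parent's logic, then applies a fixed warm cream color to the secondary mesh setting of ISM_CafeSofa.
; When the actor is destroyed, it destroys its TableCollision component.
; Begin play, overlap and tick events are empty.
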(event EventBeginPlay)

(event Collision|EventActorBeginOverlap
  (params (OtherActor :type Actor%20Object%20Reference)))

(event EventTick
  (params (DeltaSeconds :type Float%20%28single-precision%29)))

(event EventDestroyed
  (Components|DestroyComponent (Variables|Default|GetTableCollision)))

(fn ConstructionScript
  (|CallParentFunction)
  (Default|SetColoronSecondaryMesh "(R=0.846873,G=0.791298,B=0.679543,A=1.000000)" (Variables|Default|GetISM_CafeSofa)))
